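# UI material for an Undefeated item shop tile, showing a character texture on a green gradient inside a cream outline.
# The parameterized texture is scaled and shifted, then layered over an offset alpha silhouette as a drop shadow, using its RGB rather than alpha as blend factor.
# Step masks on 9-slice UVs cut the border, with thickness derived from native and actual pixel size.
# A second green-only masked blend uses scaled UVs.
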Material /ItemShop/UI/Art/Undefeated/M_UI_Undefeated_ShopTile
graph {
 properties {bUsedWithStaticMesh=false, materialDomain=MD_UI, pixelDepthOffsetMode=PDOM_Legacy, usedShadingModels=Default Lit}
 nodes [
  n1 = LinearInterpolate(
   A=Constant3Vector() {constant=<rgba 0.0239 0.4287 0 1>}
   B=Constant3Vector() {constant=<rgba 0.1912 0.8796 0.2122 1>}
   Alpha=MaterialFunctionCall(GradientStart=Constant(), GradientEnd=Constant() {r=1}, RotationAngle=Constant() {r=0.25}) {materialFunction=&MF_UI_LinearGradient}.Result
  )
  n2 = Constant3Vector() {constant=<rgba 0.0704 0.491 0.0452 1>}
  n3 = Add(
   A=MaterialFunctionCall(
    UVs=MaterialFunctionCall() {materialFunction=&MF_UI_UVScaling}.ScaleX
    Texture Scale=ScalarParameter() {defaultValue=1.5, group=Transform, parameterName=Character Size}
   ) {materialFunction=&ScaleUVsByCenter}.UVs
   B=AppendVector(
    A=ScalarParameter() {defaultValue=0.08, group=Transform, parameterName=X Value}
    B=ScalarParameter() {defaultValue=0.07, group=Transform, parameterName=Y Value}
   )
  )
  n4 = TextureObjectParameter() {
   parameterName=Texture
   texture=/OfferCatalog/Textures/BattleRoyale/FeaturedItems/Outfit/T-AthenaSoldiers-CID-739-Athena-Commando-M-CardboardCrew_01
  }
  n5 = TextureSample(UVs=Add(A=@n3) {constB=-0.02}, Tex=@n4)
  n6 = TextureSample(UVs=@n3, Tex=@n4)
  n7 = NamedRerouteUsage()
  n8 = MaterialFunctionCall() {materialFunction=&GetUserInterfaceUV}
  n9 = ComponentMask(None=@n8.9-Slice UV) {r=true}
  n10 = Step(Y=@n7, X=@n9)
  n11 = Step(Y=@n7, X=OneMinus(None=@n9))
  n12 = NamedRerouteUsage()
  n13 = ComponentMask(None=@n8.9-Slice UV) {g=true}
  n14 = Step(Y=@n12, X=@n13)
  n15 = Step(Y=@n12, X=OneMinus(None=@n13))
  LinearInterpolate(
   A=Constant3Vector() {constant=<rgba 1 0.9734 0.6724 1>}
   B=LinearInterpolate(A=LinearInterpolate(A=@n1, B=@n2, Alpha=@n5.A), B=@n6.RGB, Alpha=@n6.RGB)
   Alpha=Min(A=Min(A=@n10, B=@n11), B=Min(A=@n14, B=@n15))
  )
  n16 = NamedRerouteUsage()
  n17 = Multiply(
   A=MaterialFunctionCall() {materialFunction=&GetUserInterfaceUV}.9-Slice UV
   B=Constant2Vector() {g=0.35, r=0.35}
  )
  n18 = ComponentMask(None=@n17) {r=true}
  n19 = Step(Y=@n16, X=@n18)
  n20 = Step(Y=@n16, X=OneMinus(None=@n18))
  n21 = NamedRerouteUsage()
  n22 = ComponentMask(None=@n17) {g=true}
  n23 = Step(Y=@n21, X=@n22)
  n24 = Step(Y=@n21, X=OneMinus(None=@n22))
  LinearInterpolate(A=@n2, B=@n1, Alpha=Min(A=Min(A=@n19, B=@n20), B=Min(A=@n23, B=@n24)))
  n25 = MaterialFunctionCall(Float2=VectorParameter() {parameterName=Native Pixel Size, defaultValue=<rgba 256 256 0 1>}.RGB) {materialFunction=&BreakOutFloat2Components}
  n26 = ScalarParameter() {defaultValue=0.03, parameterName=Outline Thickness}
  NamedRerouteDeclaration(None=Divide(A=Multiply(A=@n26, B=@n25.R), B=NamedRerouteUsage())) {name=Thickness X, nodeColor=<rgba 1 0 0.5102 1>}
  NamedRerouteDeclaration(None=Divide(A=Multiply(A=@n25.G, B=@n26), B=NamedRerouteUsage())) {name=Thickness Y, nodeColor=<rgba 1 0.8901 0 1>}
  n27 = MaterialFunctionCall() {materialFunction=&GetUserInterfaceUV}
  NamedRerouteDeclaration(None=ComponentMask(None=@n27.Pixel Size) {r=true}) {name=PixelSizeX, nodeColor=<rgba 1 0.0389 0.1086 1>}
  NamedRerouteDeclaration(None=ComponentMask(None=@n27.Pixel Size) {g=true}) {name=PixelSize Y, nodeColor=<rgba 1 0.3984 0 1>}
 ]
}
