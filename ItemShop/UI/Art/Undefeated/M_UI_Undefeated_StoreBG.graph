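# UI material for the Undefeated item shop background, drawing a procedural blue-to-cyan backdrop with animated light beams.
# Two sunburst layers come from sines of radial angle, with time-driven sines swaying their rotation, thickness and horizontal scale, and radial gradients fading them.
# Their combined mask blends background and overlay radial gradients, then a fixed cyan glow is added near the bottom.
# Fill-scaled UVs feed a named reroute, and parameters expose gradient and beam tuning.
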Material /ItemShop/UI/Art/Undefeated/M_UI_Undefeated_StoreBG
graph {
 properties {bUsedWithStaticMesh=false, materialDomain=MD_UI, pixelDepthOffsetMode=PDOM_Legacy, usedShadingModels=Default Lit}
 nodes [
  n1 = NamedRerouteUsage()
  n2 = MaterialFunctionCall(
   UVs=@n1
   CenterPosition=AppendVector(
    A=ScalarParameter() {defaultValue=0.5, group=CC-Series, parameterName=X-Offset}
    B=ScalarParameter() {defaultValue=0.9, group=CC-Series, parameterName=Y-Offset}
   )
   Radius=ScalarParameter() {defaultValue=0.9, group=CC-Series, parameterName=Background Radius}
   Density=ScalarParameter() {defaultValue=1.25, group=CC-Series, parameterName=Background Density}
  ) {materialFunction=&RadialGradientExponential}
  n3 = LinearInterpolate(
   A=Constant3Vector() {constant=<rgba 0 0.0144 0.0409 1>}
   B=Constant3Vector() {constant=<rgba 0 0.162 0.3613 1>}
   Alpha=@n2.RadialGradientExponential
  )
  n4 = Constant3Vector() {constant=<rgba 0 0.6187 0.6615 1>}
  n5 = Constant3Vector() {constant=<rgba 0 0.162 0.3613 1>}
  n6 = Constant3Vector() {constant=<rgba 0 0.5733 1 1>}
  n7 = MaterialFunctionCall(
   UVs=@n1
   CenterPosition=AppendVector(
    A=ScalarParameter() {defaultValue=0.5, group=CC-Series, parameterName=Overlay X-Offset}
    B=ScalarParameter() {defaultValue=0.925, group=CC-Series, parameterName=Overlay Y-Offset}
   )
   Radius=ScalarParameter() {defaultValue=0.8, group=CC-Series, parameterName=Overlay Radius}
   Density=ScalarParameter() {defaultValue=8, group=CC-Series, parameterName=Overlay Density}
  ) {materialFunction=&RadialGradientExponential}
  n8 = ScalarParameter() {defaultValue=-0.5, group=CC-Series-Beams, parameterName=1st-Beam X-Offset}
  n9 = ScalarParameter() {defaultValue=-0.9, group=CC-Series-Beams, parameterName=1st-Beam Y-Offset}
  n10 = MaterialFunctionCall(
   CenterPosition=Multiply(A=AppendVector(A=@n8, B=@n9)) {constB=-1}
   Radius=ScalarParameter() {defaultValue=1.5, group=CC-Series-Beams, parameterName=1st-Beam-Radius}
   Density=ScalarParameter() {defaultValue=0.75, group=CC-Series-Beams, parameterName=1st-Beam-Density}
  ) {materialFunction=&RadialGradientExponential}
  n11 = AppendVector(
   A=ScalarParameter() {defaultValue=-0.5, group=CC-Series-Beams, parameterName=1st-Beam X-Offset}
   B=ScalarParameter() {defaultValue=-0.9, group=CC-Series-Beams, parameterName=1st-Beam Y-Offset}
  )
  n12 = NamedRerouteUsage()
  n13 = Sine(None=Multiply(A=Time(), B=ScalarParameter() {defaultValue=0.5, group=CC-Series-Beams, parameterName=Beam-Scale-TimeFactor})) {period=2}
  n14 = AppendVector(A=LinearInterpolate(Alpha=Add(A=@n13) {constB=0.5}) {constA=1, constB=1.05}, B=Constant() {r=0.5})
  n15 = Sine(None=Multiply(A=Time(), B=ScalarParameter() {defaultValue=0.1, group=CC-Series-Beams, parameterName=1st-Beam-TimeFactor})) {period=2}
  n16 = Rotator(Coordinate=Multiply(A=Add(A=@n11, B=@n12), B=@n14), Time=LinearInterpolate(Alpha=@n15) {constA=0.5}) {centerX=0, centerY=0}
  n17 = Sine(None=Multiply(
    A=MaterialFunctionCall(Vector or UVs=@n16) {materialFunction=&VectorToRadialValue}.Vector Converted to Angle
    B=ScalarParameter() {defaultValue=20, group=CC-Series-Beams, parameterName=1st-Beam-Count}
   ) {constB=2})
  n18 = LinearInterpolate(
   A=ScalarParameter() {group=CC-Series-Beams, parameterName=1st-Beam-Thickness-Min}
   B=ScalarParameter() {defaultValue=0.2, group=CC-Series-Beams, parameterName=1st-Beam-Thickness-Max}
   Alpha=@n15
  )
  n19 = Saturate(None=Multiply(A=Add(A=@n17, B=@n18) {constB=-0.25}, B=ScalarParameter() {defaultValue=1, group=CC-Series-Beams, parameterName=1st-Beam-Edge}))
  n20 = LinearInterpolate(B=Add(A=Multiply(A=@n17, B=@n10.RadialGradientExponential), B=Multiply(A=@n19, B=@n10.RadialGradientExponential)), Alpha=@n19)
  n21 = ScalarParameter() {defaultValue=1.5, group=CC-Series-Beams, parameterName=1st-Beam-Multiply}
  n22 = AppendVector(
   A=ScalarParameter() {defaultValue=-0.5, group=CC-Series-Beams, parameterName=2nd-Beam X-Offset}
   B=ScalarParameter() {defaultValue=-0.9, group=CC-Series-Beams, parameterName=2nd-Beam Y-Offset}
  )
  n23 = Sine(None=Multiply(A=Time(), B=ScalarParameter() {defaultValue=0.125, group=CC-Series-Beams, parameterName=2nd-Beam-TimeFactor})) {period=2}
  n24 = Rotator(Coordinate=Multiply(A=@n14, B=Add(A=@n12, B=@n22)), Time=LinearInterpolate(Alpha=@n23) {constA=0.25, constB=2}) {centerX=0, centerY=0}
  n25 = Sine(None=Multiply(
    A=MaterialFunctionCall(Vector or UVs=@n24) {materialFunction=&VectorToRadialValue}.Vector Converted to Angle
    B=ScalarParameter() {defaultValue=12, group=CC-Series-Beams, parameterName=2nd-Beam-Count}
   ) {constB=2})
  n26 = LinearInterpolate(
   A=ScalarParameter() {group=CC-Series-Beams, parameterName=2nd-Beam-Thickness-Min}
   B=ScalarParameter() {defaultValue=0.1, group=CC-Series-Beams, parameterName=2nd-Beam-Thickness-Max}
   Alpha=@n23
  )
  n27 = Saturate(None=Multiply(A=Add(A=@n25, B=@n26) {constB=-0.25}, B=ScalarParameter() {defaultValue=1, group=CC-Series-Beams, parameterName=2nd-Beam-Edge}))
  n28 = ScalarParameter() {defaultValue=-0.5, group=CC-Series-Beams, parameterName=2nd-Beam X-Offset}
  n29 = ScalarParameter() {defaultValue=-0.9, group=CC-Series-Beams, parameterName=2nd-Beam Y-Offset}
  n30 = MaterialFunctionCall(
   CenterPosition=Multiply(A=AppendVector(A=@n28, B=@n29)) {constB=-1}
   Radius=ScalarParameter() {defaultValue=1.5, group=CC-Series-Beams, parameterName=2nd-Beam-Radius}
   Density=ScalarParameter() {defaultValue=0.75, group=CC-Series-Beams, parameterName=2nd-Beam-Density}
  ) {materialFunction=&RadialGradientExponential}
  n31 = Multiply(A=@n27, B=@n30.RadialGradientExponential)
  n32 = Multiply(A=@n25, B=@n30.RadialGradientExponential)
  n33 = Saturate(None=LinearInterpolate(B=Add(A=@n31, B=@n32), Alpha=@n27))
  n34 = Saturate(None=LinearInterpolate(A=Multiply(A=@n20, B=@n21), B=ScalarParameter() {defaultValue=3, group=CC-Series-Beams, parameterName=2nd-Beam-Multiply}, Alpha=@n33))
  n35 = NamedRerouteUsage()
  n36 = Constant() {r=0.55}
  LinearInterpolate(
   A=LinearInterpolate(
    A=LinearInterpolate(A=@n3, B=@n4, Alpha=@n2.RadialGradientExponential)
    B=LinearInterpolate(A=@n5, B=@n6, Alpha=@n7.RadialGradientExponential)
    Alpha=Add(A=@n33, B=@n34)
   )
   B=Constant3Vector() {constant=<rgba 0 0.7918 0.9097 1>}
   Alpha=MaterialFunctionCall(
    UVs=MaterialFunctionCall(UVs=@n35, Texture Scale=@n36) {materialFunction=&ScaleUVsByCenter}.UVs
    CenterPosition=Constant2Vector() {g=1.25, r=0.5}
    Radius=Constant() {r=0.6}
    Density=Constant() {r=2}
   ) {materialFunction=&RadialGradientExponential}.RadialGradientExponential
  )
  MaterialFunctionCall(Float2=TextureCoordinate() {coordinateIndex=3}) {materialFunction=&BreakOutFloat2Components}
  NamedRerouteDeclaration(None=MaterialFunctionCall(
    UVs=MaterialFunctionCall() {materialFunction=&MF_UI_UVScaling}.Scaled To Fill
    Texture Scale=Constant() {r=0.7}
   ) {materialFunction=&ScaleUVsByCenter}.UVs) {name=UVScaling, nodeColor=<rgba 1 0 0.2812 1>}
 ]
}
